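; Scripted holiday tree prop Blueprint that can play music through a music party controller component.
; Its construction script calls the parent, then auto-activates that component only when both the music controller class and speaker variables are valid.
; It copies the speaker asset and controller class onto the component regardless of that check.
; BeginPlay only calls the parent, and the overlap and tick events are disabled stubs.
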(event EventBeginPlay
  (|CallParentFunction))

(event Collision|EventActorBeginOverlap
  (params (OtherActor :type Actor%20Object%20Reference))
  (|CallParentFunction OtherActor) ; This node is disabled and will not be called.\nDrag off pins to build functionality.
)

(event EventTick
  (params (DeltaSeconds :type Float%20%28single-precision%29))
  (|CallParentFunction DeltaSeconds) ; This node is disabled and will not be called.\nDrag off pins to build functionality.
)

(fn ConstructionScript
  (bind _musicpartypropcomponent_controller (Variables|Default|GetMusicPartyPropComponent_Controller))
  (|CallParentFunction)
  (bind _output_get (LocalVariables|SetPlayMusic (and (Utilities|IsValidClass (Variables|Music|GetMusicController)) (Utilities|IsValid (Variables|Music|GetMusicSpeaker) :!class /Script/Engine.KismetSystemLibrary))))
  (Components|Activation|SetAutoActivate (Variables|Default|GetMusicPartyPropComponent_Controller) _output_get)
  (Variables|/Game/Sounds/Fort_Audio_Blueprints/MusicParty/MusicPartyPropComponent_SpeakerAssets.MusicPartyPropComponent_SpeakerAssets_C|SetSpeakerAsset (Variables|Music|GetMusicSpeaker) _musicpartypropcomponent_controller)
  (Variables|/Game/Sounds/Fort_Audio_Blueprints/MusicParty/MusicPartyPropComponent_Controller.MusicPartyPropComponent_Controller_C|SetControllerClass (Variables|Music|GetMusicController) _musicpartypropcomponent_controller))
